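; BP_HeavyGavel_PIllar_B is a pillar prop Blueprint in the HeavyGavel environment set.
; Its construction script checks an override-static-mesh-mobility flag.
; When the flag is off, it loops over all StaticMeshComponents on the actor and sets each to Static mobility.
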(fn ConstructionScript
  (if (not (Variables|Mesh|GetOverrideStaticMeshMobility_0))
    (for _array_element (Actor|GetComponentsByClass "/Script/Engine.StaticMeshComponent")
      (Transformation|SetMobility _array_element "Static"))))
